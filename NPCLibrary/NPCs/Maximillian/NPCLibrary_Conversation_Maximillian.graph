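# Conversation database for the NPC Maximillian, a self-styled ace pilot, defining the speech shown when a player talks to them.
# The entry tag leads to one Speech node with a default line and contextual variants for a first-ever conversation or specific player character skins.
# The first-meeting line takes its priority from a curve row, while skin lines use zero.
# The speech then links out to the shared NPCLibrary Quests and BasicServices conversation entries.
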ConversationDatabase /NPCLibrary/NPCs/Maximillian/NPCLibrary_Conversation_Maximillian
entry_points [{tag=Conversation.Entry.Maximillian, destinations [#1]}]
nodes [
 ConversationEntryPointNode_7 {
  id=#1
  class=ConversationEntryPointNode
  outputs [#2]
  properties {entryTag {tagName=Conversation.Entry.Maximillian}}
 }
 Speech_C_4 {
  id=#2
  class=Speech_C
  outputs [#3, #4]
  properties {
   nodeName=""
   generalConfig {
    defaultMessage="I own my own plane, you know."
    contextualMessages [
     {
      message="Hello! I'm Maximillian, ace pilot and world traveler. I also cosplay."
      contextRequirements [{requirement {class=FortControllerRequirement_IsFirstEverConversationWithNPC}}]
      priority {curve {rowName=Tandem.Speech.Priority.FirstEverInteraction}}
     }
     {
      message="Hail, fellow traveller! Wait, I think we met at PlaneCon?"
      contextRequirements [{requirement {
         class=FortControllerRequirement_HasCID
         properties {allowedCIds [/BRCosmetics/Athena/Items/Cosmetics/Characters/CID_224_Athena_Commando_F_Dieselpunk]}
        }}]
      priority {value=0, curve {curveTable=null, rowName=None}}
     }
     {
      message="That suit is very impressive, but can it fly? No? I didn't think so."
      contextRequirements [{requirement {class=FortControllerRequirement_HasCID, properties {allowedCIds [&Character_BadBear]}}}]
      priority {value=0, curve {curveTable=null, rowName=None}}
     }
     {
      message="You'll never get away with this!"
      contextRequirements [{requirement {class=FortControllerRequirement_HasCID, properties {allowedCIds [&Character_TheHerald]}}}]
      priority {value=0, curve {curveTable=null, rowName=None}}
     }
     {
      message=I admire the Royale Air Force.
      requirementMatchPolicy=RequireAny
      contextRequirements [{requirement {
         class=FortControllerRequirement_HasCID
         properties {allowedCIds [
           /BRCosmetics/Athena/Items/Cosmetics/Characters/CID_031_Athena_Commando_M_Retro
           /BRCosmetics/Athena/Items/Cosmetics/Characters/CID_412_Athena_Commando_F_Raptor
          ]}
        }}]
      priority {value=0, curve {curveTable=null, rowName=None}}
     }
     {
      message="Please do not \"ollie\" off my propeller. Thanks."
      contextRequirements [{requirement {class=FortControllerRequirement_HasCID, properties {allowedCIds [&Character_ChillCat]}}}]
      priority {value=0, curve {curveTable=null, rowName=None}}
     }
     {
      message=I could fly you to safety... but you bring your danger with you.
      contextRequirements [{requirement {class=FortControllerRequirement_HasCID, properties {allowedCIds [&Character_Bites]}}}]
      priority {value=0, curve {curveTable=null, rowName=None}}
     }
     {
      message=Anywhere you need to fly?
      contextRequirements [{requirement {class=FortControllerRequirement_HasCID, properties {allowedCIds [&Character_PinkSpike]}}}]
      priority {value=0, curve {curveTable=null, rowName=None}}
     }
     {
      message="I don't think that suit will fit in my plane, small guy."
      contextRequirements [{requirement {class=FortControllerRequirement_HasCID, properties {allowedCIds [&Character_BadBear]}}}]
      priority {value=0, curve {curveTable=null, rowName=None}}
     }
     {
      message=I knew someone like you once. Only he wore all black. And talked to himself.
      contextRequirements [{requirement {class=FortControllerRequirement_HasCID, properties {allowedCIds [&Character_RoseDust]}}}]
      priority {value=0, curve {curveTable=null, rowName=None}}
     }
    ]
   }
  }
 }
 ConversationLinkNode_0 {id=#4, class=ConversationLinkNode, properties {remoteEntryTag {tagName=Conversation.Entry.BasicServices}}}
 ConversationLinkNode_1 {id=#3, class=ConversationLinkNode, properties {remoteEntryTag {tagName=Conversation.Entry.NPCLibrary.Quests}}}
]
